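# Physics asset for the Blue Flames backpack cosmetic mesh, defining collision and joints for its two dynamic hoses.
# A capsule body covers the pack, and small sphere bodies sit on the left and right hose bones.
# Each hose is constrained to the pack with matching limits that allow modest swing and very little twist.
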PhysicsAsset /BRCosmetics/Accessories/FORT_Backpacks/Backpack_F_MED_Blue_Flames/Meshes/F_MED_Blue_Flames_Pack_PhysicsAsset
bodies {
 dyn_hose_l [{shapeType=Sphere, radius=5, center=<2 0 0>}]
 dyn_hose_r [{shapeType=Sphere, radius=5, center=<2 0 0>}]
 pack [{shapeType=Capsule, length=16.9422, radius=9.2476, center=<-0.0689 -6.1494 -6.3679>, rotation=<rot 0 -89.9443 -89.6331>}]
}
constraints [
 {bone1Name=dyn_hose_r, bone2Name=pack, swing1LimitDegrees=20, swing2LimitDegrees=20, twistLimitDegrees=2}
 {bone1Name=dyn_hose_l, bone2Name=pack, swing1LimitDegrees=20, swing2LimitDegrees=20, twistLimitDegrees=2}
]
